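# Parent material for the EctoCat pickaxe, built on the 2023 character parent PreFX and PostFX functions, animating their emissive output.
# Panning cloud noise distorts the UVs of a second panning sample of the same texture, which then mottles the emissive.
# A slow sine pulse varies the brightness, and a time-of-day function dims it toward night.
# All other material attributes pass through unchanged.
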Material /BRCosmetics/Weapons/FORT_Melee/Pickaxe_EctoCat/Materials/M_EctoCat_Pickaxe_PARENT
graph {
 properties {
  bUseMaterialAttributes=true
  bUsedWithMeshDeformer=true
  bUsedWithNanite=true
  bUsedWithSkeletalMesh=true
  pixelDepthOffsetMode=PDOM_Legacy
  shadingModel=MSM_FromMaterialExpression
  usedShadingModels=Default Lit | Subsurface | Clear Coat | Subsurface Profile | Cloth
 }
 nodes [
  n1 = MaterialFunctionCall() {materialFunction=&MF_CharacterParent_2023_PreFX}
  n2 = GetMaterialAttributes(None=MaterialFunctionCall(Material=@n1.Result, M=@n1.Result, Pre Skinned Local Position=@n1.Result, Distance Blend=@n1.Result, Fresnel=@n1.Result) {materialFunction=&MF_CharacterParent_2023_PostFX}.Result) {attributeGetTypes [#1]}
  n3 = Multiply(A=TextureCoordinate()) {constB=1.3}
  n4 = TextureSample(UVs=Panner() {speedX=0.017, speedY=0.024}) {texture=&AtmosphericCloudNoise01}
  n5 = TextureSample(UVs=Add(A=Panner(Coordinate=@n3) {speedX=0.06, speedY=0.017}, B=Multiply(A=@n4.R) {constB=0.18})) {texture=&AtmosphericCloudNoise01}
  n6 = Multiply(A=@n2.Emissive Color, B=Saturate(None=Add(A=@n5.R) {constB=0.27}))
  n7 = Sine(None=Multiply(A=Time()) {constB=0.3})
  n8 = LinearInterpolate(Alpha=Add(A=Multiply(A=@n7) {constB=0.5}) {constB=0.5}) {constA=1, constB=1.7}
  n9 = Constant() {r=1}
  n10 = Constant() {r=0.9}
  n11 = Constant() {r=0.55}
  SetMaterialAttributes(
   MaterialAttributes=@n2.MaterialAttributes
   Emissive Color=Multiply(
    A=Multiply(A=@n6, B=@n8)
    B=MaterialFunctionCall(Value_Morning=@n9, Value_Day=@n9, Value_Evening=@n10, Value_Night=@n11) {materialFunction=&Scalar_Time_of_Day}.Result
   )
  ) {attributeSetTypes [#1]}
 ]
}
